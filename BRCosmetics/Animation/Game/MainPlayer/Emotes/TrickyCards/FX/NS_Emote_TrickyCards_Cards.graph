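# Niagara effect for the TrickyCards emote that throws playing-card meshes from the character's right hand.
# A Position emitter tracks hand_r, feeding Cards001, a tapering stream, and Final_Card_Burst001, a timed closing burst.
# Cards pick random mesh variants and tumble with gravity, wind and aerodynamic drag, then collide, align to surfaces and shrink away.
# User.Card_Velocity narrows the launch cone as it rises and also scales the stream's speed.
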NiagaraSystem /BRCosmetics/Animation/Game/MainPlayer/Emotes/TrickyCards/FX/NS_Emote_TrickyCards_Cards
dependencies {
 usedDataInterfaces [
  &NiagaraDataInterfaceSkeletalMesh
  &NiagaraDataInterfaceCurve
  &NiagaraDataInterfaceParticleRead
  &NiagaraDataInterfaceSpriteRendererInfo
 ]
 usedDynamicInputs []
 usedModules [&SystemState]
}
emitters {
 Cards001 {
  emitterUpdateScript [
   &EmitterState(Life Cycle Mode=::NewEnumerator1, Loop Behavior=::NewEnumerator1, Loop Duration=6.3, UseLoopDelay=false)
   &SpawnRate(
    SpawnRate=&FloatFromCurve(
     FloatCurve=DataInterface() {propertyValues {Curve {keys [<at 0 1.5>,<at 0.85 1>,<at 0.9 0>]}}}
     CurveIndex=$Emitter.NormalizedLoopAge
     Scale Curve=35
    )
    Use Spawn Probability=false
   ) {enabled=false}
   &SpawnParticlesFromOtherEmitter(
    Attribute Reader=DataInterface() {propertyValues {EmitterBinding {bindingMode=Other, emitterName=Position}}}
    Spawn Rate=&FloatFromCurve(
     FloatCurve=DataInterface() {propertyValues {Curve {keys [<at 0 1.5>,<at 0.85 1>,<at 0.9 0>]}}}
     CurveIndex=$Emitter.NormalizedLoopAge
     Scale Curve=35
    )
    Use Maximum Age=false
    Use Minimum Age=false
   )
  ]
  particleSpawnScript [
   /SetVariables(Particles.MeshIndex=&RandomRangeInt(Minimum=0, Maximum=7)) {isSetParameters=true}
   &InitializeParticle(
    Lifetime Min=1.5
    Lifetime Mode=::NewEnumerator1
    Mass Max=2
    Mass Mode=::NewEnumerator3
    Mesh Scale Mode=::NewEnumerator3
    Sprite Size=<7 12>
    Sprite Size Mode=::NewEnumerator4
    UsePositionOffset=false
    Write Material Random=true
   )
   &SkeletalMeshLocation(
    Bone Sampling Mode=::NewEnumerator4
    MaskSpawnGroup=false
    Orientation Sampling=::NewEnumerator1
    Position Sampling=::NewEnumerator2
    Sampled Position Offset Coordinate Space:ENiagaraSkelMeshTransforms=::NewEnumerator0
    Skeletal Mesh=DataInterface() {propertyValues {
      DefaultMesh=null
      ExcludeBoneName=root
      OptionalComponentClass=null
      PreviewMesh=null
      SkinningMode=SkinOnTheFly
      SoftSourceActor=null
      SourceMode=AttachParent
      UvSetIndex=0
      WholeMeshLOD=-1
      bExcludeBone=false
      bReadDeformedGeometry=true
      bRequireCurrentFrameData=true
      ComponentTags []
      FilteredBones [hand_r]
      FilteredSockets []
      MeshUserParameter {parameter {name=None, typeDefHandle {registeredTypeIndex=1}, varData []}}
      SamplingRegions []
     }}
   )
   &AddVelocity(
    Cone Angle=&FloatFromCurve(FloatCurve=DataInterface() {propertyValues {Curve {keys [<at 0 1>,<at 1 0>]}}}, CurveIndex=$User.Card_Velocity, Scale Curve=300)
    Cone Axis=&Multiply_VectorByFloat(
     Vector=&ReturnMeshOrientationAxis(Mesh Orientation Axis=::XAxis, Mesh Orientation Quat=$Particles.SkeletalMeshLocation.SampledRotation)
     Float=-1
    )
    Distribution Along Cone Axis=0.2
    Module.UseConeFalloff=false
    Module.UseVelDistribution=true
    Override Seed=false
    Rotation Coordinate Space=::Simulation
    Velocity Mode=::NewEnumerator2
    Velocity Speed=&Multiply_Float(
     A=&FloatFromCurve(FloatCurve=DataInterface() {propertyValues {Curve {keys [<at 0 0>,<at 1 1>]}}}, CurveIndex=$User.Card_Velocity)
     B=&Multiply_Float(A=&RandomRangeFloat(Minimum=300, Maximum=650), B=0.7)
    )
   )
   &SampleParticlesFromOtherEmitter()
   &InitialMeshOrientation(Use Rotation Vector=false)
  ]
  particleUpdateScript [
   &ParticleState()
   &ScaleMeshSize(Scale Factor=&VectorFromFloat(Value=&FloatFromCurve(FloatCurve=DataInterface() {propertyValues {Curve {keys [
          <at -0.0046 0 RCIM_Cubic>
          {arriveTangent=0.7374, interpMode=RCIM_Cubic, leaveTangent=0.7374, tangentMode=RCTM_User, time=0.01, value=0}
          {arriveTangent=-0.0641, interpMode=RCIM_Cubic, leaveTangent=-0.0641, tangentMode=RCTM_User, time=0.02, value=1}
          {arriveTangent=0.0602, interpMode=RCIM_Cubic, leaveTangent=0.0602, tangentMode=RCTM_User, time=0.8, value=1}
          {arriveTangent=-3.3518, interpMode=RCIM_Cubic, leaveTangent=-3.3518, tangentMode=RCTM_User, time=1, value=0}
         ]}}})))
   &GravityForce()
   &WindForce(
    Offset By Time:Vector3f=<50 10 -125>
    Scale Min / Max:Vector2f=<0.275 1>
    Speed Min / Max:Vector2f=<0 500>
    Turbulence Scale Mode:ENiagaraWindTurbulenceContributionMode=::NewEnumerator1
    Wind Speed=<0 0 0>
    Wind Speed Scale=&RandomRangeFloat(Minimum=0.333, Maximum=1)
   )
   &AerodynamicDrag(
    Aerodynamic Drag:NiagaraFloat=&RandomRangeFloat(Minimum=1, Maximum=0.1)
    Aerodynamic Rotational Drag=0.15
    Drag Shape Curve=DataInterface() {propertyValues {Curve {keys [
        {interpMode=RCIM_Cubic, tangentMode=RCTM_User, time=0, value=0.1}
        {arriveTangent=0.9, interpMode=RCIM_Cubic, leaveTangent=0.9, time=0.5, value=0}
        {arriveTangent=2.0799, interpMode=RCIM_Cubic, leaveTangent=2.0799, tangentMode=RCTM_User, time=1, value=1}
       ]}}}
    Lift Contribution=0
    Lift Shape Curve=DataInterface() {propertyValues {Curve {keys [
        {interpMode=RCIM_Cubic, tangentMode=RCTM_User, time=0, value=0}
        {interpMode=RCIM_Cubic, tangentMode=RCTM_User, time=0.5, value=1}
        {tangentMode=RCTM_User, time=1, value=0}
       ]}}}
    Pivot Offset Maximum:Vector3f=<5 0.5 1>
    Pivot Offset Minimum:Vector3f=<-5 -0.5 -1>
    Rotation Strength:NiagaraFloat=&RandomRangeFloat(Minimum=1.8, Maximum=3.2)
    Use Linear Drag=true
    Use Rotational Drag=true
   )
   &Collision(
    Advanced Aging Rate=5
    Age Colliding Particles=true
    Collision Enabled=&SetBoolByFloatComparison(A=$Particles.NormalizedAge, B=0.45)
    Enable Rest State=false
    Friction=2
    Friction During a Bounce=0.1
    KillOnCollision=false
    Module.EnableMaxCollisionCount=false
    Module.Randomize Collision Normal Vector=false
    Particle Radius Scale=2
    Restitution=0.5
   )
   &Drag(
    Drag=&Multiply_Float(A=&RandomRangeFloat(Minimum=0.8, Maximum=1.2), B=2)
    Use Linear Drag=true
    Use Rotational Drag=false
   )
   &SolveForcesAndVelocity(Clamp Velocity=false, Limit Acceleration=false)
   &AlignSpriteToMeshOrientation(SetAlignment=true, SetFacing=true)
   /Niagara/Modules/Collision/AlignParticlesWithCollisionPlane(
    Allow Alignment on Y=false
    Allow Alignment on Z=false
    Calming Rate When Bouncing=0
    Calming Rate When Resting=5
    Collision Normal's Interpolation Rate=5
    Rotational Dampening Power=5
    Set the Sprite Alignment and Sprite Facing Variables=false
   )
   &SubUVAnimation(
    "Sprite Renderer "=DataInterface() {propertyValues {SpriteRenderer=&self:Cards001_3.NiagaraSpriteRendererProperties_0}}
    Start Frame Offset=49
    SubUV Lookup Index Scale=0
    UseEndFrame=false
    UseStartFrame=false
   )
   &ScaleColor(
    Scale Alpha=&FloatFromCurve(FloatCurve=DataInterface() {propertyValues {Curve {keys [
         <at -0.0046 0 RCIM_Cubic>
         {arriveTangent=0.7374, interpMode=RCIM_Cubic, leaveTangent=0.7374, tangentMode=RCTM_User, time=0.02, value=0}
         {arriveTangent=-0.0641, interpMode=RCIM_Cubic, leaveTangent=-0.0641, tangentMode=RCTM_User, time=0.03, value=1}
         {arriveTangent=0.0602, interpMode=RCIM_Cubic, leaveTangent=0.0602, tangentMode=RCTM_User, time=0.8, value=1}
         {arriveTangent=-3.3518, interpMode=RCIM_Cubic, leaveTangent=-3.3518, tangentMode=RCTM_User, time=1, value=0}
        ]}}})
    ScaleA=true
    ScaleRGB=true
   ) {enabled=false}
  ]
  rendererClasses [&NiagaraMeshRendererProperties]
  renderers [{rendererClass=&NiagaraMeshRendererProperties, rendererIndex=0}]
 }
 Final_Card_Burst001 {
  emitterUpdateScript [
   &EmitterState(Life Cycle Mode=::NewEnumerator1, Loop Behavior=::NewEnumerator1, Loop Duration=9.9, UseLoopDelay=false)
   &SpawnBurst_Instantaneous(Spawn Count=15, Spawn Time=7.1, Use Spawn Probability=false, UseLoopCountLimit=false)
   &SpawnParticlesFromOtherEmitter(
    Attribute Reader=DataInterface() {propertyValues {EmitterBinding {bindingMode=Other, emitterName=Position}}}
    Spawn Rate=&FloatFromCurve(
     FloatCurve=DataInterface() {propertyValues {Curve {keys [<at 0 1.5>,<at 0.85 1>,<at 0.9 0>]}}}
     CurveIndex=$Emitter.NormalizedLoopAge
     Scale Curve=0
    )
    Use Maximum Age=false
    Use Minimum Age=false
   )
  ]
  particleSpawnScript [
   /SetVariables(Particles.MeshIndex=&RandomRangeInt(Minimum=0, Maximum=7)) {isSetParameters=true}
   &InitializeParticle(
    Lifetime Max=1.25
    Lifetime Mode=::NewEnumerator1
    Mass Max=2
    Mass Mode=::NewEnumerator3
    Mesh Scale Mode=::NewEnumerator3
    Sprite Size=<7 12>
    Sprite Size Mode=::NewEnumerator4
    UsePositionOffset=false
    Write Material Random=true
   )
   &SkeletalMeshLocation(
    Bone Sampling Mode=::NewEnumerator4
    MaskSpawnGroup=false
    Orientation Sampling=::NewEnumerator1
    Position Sampling=::NewEnumerator2
    Sampled Position Offset Coordinate Space:ENiagaraSkelMeshTransforms=::NewEnumerator0
    Skeletal Mesh=DataInterface() {propertyValues {
      DefaultMesh=null
      ExcludeBoneName=root
      OptionalComponentClass=null
      PreviewMesh=null
      SkinningMode=SkinOnTheFly
      SoftSourceActor=null
      SourceMode=AttachParent
      UvSetIndex=0
      WholeMeshLOD=-1
      bExcludeBone=false
      bReadDeformedGeometry=true
      bRequireCurrentFrameData=true
      ComponentTags []
      FilteredBones [hand_r]
      FilteredSockets []
      MeshUserParameter {parameter {name=None, typeDefHandle {registeredTypeIndex=1}, varData []}}
      SamplingRegions []
     }}
   )
   &AddVelocity(
    Cone Angle=&FloatFromCurve(FloatCurve=DataInterface() {propertyValues {Curve {keys [<at 0 1>,<at 1 0>]}}}, CurveIndex=$User.Card_Velocity, Scale Curve=300)
    Cone Axis=&Multiply_VectorByFloat(
     Vector=&ReturnMeshOrientationAxis(Mesh Orientation Axis=::XAxis, Mesh Orientation Quat=$Particles.SkeletalMeshLocation.SampledRotation)
     Float=-1
    )
    Distribution Along Cone Axis=0.2
    Module.UseConeFalloff=false
    Module.UseVelDistribution=true
    Override Seed=false
    Rotation Coordinate Space=::Simulation
    Velocity Mode=::NewEnumerator2
    Velocity Speed=&Multiply_Float(A=&RandomRangeFloat(Minimum=300, Maximum=600), B=0.7)
   )
   &SampleParticlesFromOtherEmitter()
   &InitialMeshOrientation(Use Rotation Vector=false)
  ]
  particleUpdateScript [
   &ParticleState()
   &ScaleMeshSize(Scale Factor=&VectorFromFloat(Value=&FloatFromCurve(FloatCurve=DataInterface() {propertyValues {Curve {keys [
          <at -0.0046 0 RCIM_Cubic>
          {arriveTangent=0.7374, interpMode=RCIM_Cubic, leaveTangent=0.7374, tangentMode=RCTM_User, time=0.01, value=0}
          {arriveTangent=-0.0641, interpMode=RCIM_Cubic, leaveTangent=-0.0641, tangentMode=RCTM_User, time=0.02, value=1}
          {arriveTangent=0.0602, interpMode=RCIM_Cubic, leaveTangent=0.0602, tangentMode=RCTM_User, time=0.8, value=1}
          {arriveTangent=-3.3518, interpMode=RCIM_Cubic, leaveTangent=-3.3518, tangentMode=RCTM_User, time=1, value=0}
         ]}}})))
   &GravityForce()
   &WindForce(
    Offset By Time:Vector3f=<50 10 -125>
    Scale Min / Max:Vector2f=<0.275 1>
    Speed Min / Max:Vector2f=<0 500>
    Turbulence Scale Mode:ENiagaraWindTurbulenceContributionMode=::NewEnumerator1
    Wind Speed=<0 0 0>
    Wind Speed Scale=&RandomRangeFloat(Minimum=0.333, Maximum=1)
   )
   &AerodynamicDrag(
    Aerodynamic Drag:NiagaraFloat=&RandomRangeFloat(Minimum=1, Maximum=0.1)
    Aerodynamic Rotational Drag=0.15
    Drag Shape Curve=DataInterface() {propertyValues {Curve {keys [
        {interpMode=RCIM_Cubic, tangentMode=RCTM_User, time=0, value=0.1}
        {arriveTangent=0.9, interpMode=RCIM_Cubic, leaveTangent=0.9, time=0.5, value=0}
        {arriveTangent=2.0799, interpMode=RCIM_Cubic, leaveTangent=2.0799, tangentMode=RCTM_User, time=1, value=1}
       ]}}}
    Lift Contribution=0
    Lift Shape Curve=DataInterface() {propertyValues {Curve {keys [
        {interpMode=RCIM_Cubic, tangentMode=RCTM_User, time=0, value=0}
        {interpMode=RCIM_Cubic, tangentMode=RCTM_User, time=0.5, value=1}
        {tangentMode=RCTM_User, time=1, value=0}
       ]}}}
    Pivot Offset Maximum:Vector3f=<5 0.5 1>
    Pivot Offset Minimum:Vector3f=<-5 -0.5 -1>
    Rotation Strength:NiagaraFloat=&RandomRangeFloat(Minimum=1.8, Maximum=3.2)
    Use Linear Drag=true
    Use Rotational Drag=true
   )
   &Collision(
    Age Colliding Particles=false
    Collision Enabled=&SetBoolByFloatComparison(A=$Particles.NormalizedAge, B=0.5)
    Enable Rest State=false
    Friction=2
    Friction During a Bounce=0.1
    KillOnCollision=false
    Module.EnableMaxCollisionCount=false
    Module.Randomize Collision Normal Vector=false
    Particle Radius Scale=2
    Restitution=0.5
   )
   &Drag(
    Drag=&Multiply_Float(A=&RandomRangeFloat(Minimum=0.8, Maximum=1.2), B=3)
    Use Linear Drag=true
    Use Rotational Drag=false
   )
   &SolveForcesAndVelocity(Clamp Velocity=false, Limit Acceleration=false)
   &AlignSpriteToMeshOrientation(SetAlignment=true, SetFacing=true)
   &AlignParticlesWithCollisionPlane(
    Allow Alignment on Y=false
    Allow Alignment on Z=false
    Calming Rate When Resting=5
    Collision Normal's Interpolation Rate=5
    Rotational Dampening Power=5
    Set the Sprite Alignment and Sprite Facing Variables=false
   )
   &SubUVAnimation(
    "Sprite Renderer "=DataInterface() {propertyValues {SpriteRenderer=&self:Final_Card_Burst001_1.NiagaraSpriteRendererProperties_0}}
    Start Frame Offset=49
    SubUV Lookup Index Scale=0
    UseEndFrame=false
    UseStartFrame=false
   )
   &ScaleColor(
    Scale Alpha=&FloatFromCurve(FloatCurve=DataInterface() {propertyValues {Curve {keys [
         <at -0.0046 0 RCIM_Cubic>
         {arriveTangent=0.7374, interpMode=RCIM_Cubic, leaveTangent=0.7374, tangentMode=RCTM_User, time=0.02, value=0}
         {arriveTangent=-0.0641, interpMode=RCIM_Cubic, leaveTangent=-0.0641, tangentMode=RCTM_User, time=0.03, value=1}
         {arriveTangent=-1.0309, interpMode=RCIM_Cubic, leaveTangent=-1.0309, time=0.5, value=0.9}
         {arriveTangent=0.1444, interpMode=RCIM_Cubic, leaveTangent=0.1444, tangentMode=RCTM_User, time=1, value=0}
        ]}}})
    ScaleA=true
    ScaleRGB=true
   ) {enabled=false}
  ]
  rendererClasses [&NiagaraMeshRendererProperties]
  renderers [{rendererClass=&NiagaraMeshRendererProperties, rendererIndex=0}]
 }
 Position {
  emitterUpdateScript [
   &EmitterState(Life Cycle Mode=::NewEnumerator1, Loop Behavior=::NewEnumerator1, Loop Duration=1, UseLoopDelay=false)
   &SpawnBurst_Instantaneous(Use Spawn Probability=false, UseLoopCountLimit=false)
  ]
  particleSpawnScript [&InitializeParticle(
    Lifetime=14.8
    Mass Max=2
    Mass Mode=::NewEnumerator3
    Sprite Size Max=<3 14>
    Sprite Size Min=<2 6>
    Sprite Size Mode=::NewEnumerator2
    UsePositionOffset=false
    Write Material Random=true
   )]
  particleUpdateScript [
   &ParticleState()
   &SkeletalMeshLocation(
    Bone Sampling Mode=::NewEnumerator4
    MaskSpawnGroup=false
    Orientation Sampling=::NewEnumerator1
    Sampled Position Offset=&Lerp_Vector(
     A=<-10 0 0>
     B=<-15 0 0>
     Alpha=&FloatFromCurve(
      FloatCurve=DataInterface() {propertyValues {Curve {keys [<at 0 0 RCIM_Constant>,<at 0.5 1 RCIM_Constant>,<at 1 1 RCIM_Constant>]}}}
      CurveIndex=$Emitter.NormalizedLoopAge
     )
    )
    Skeletal Mesh=DataInterface() {propertyValues {
      DefaultMesh=null
      ExcludeBoneName=root
      OptionalComponentClass=null
      PreviewMesh=null
      SkinningMode=SkinOnTheFly
      SoftSourceActor=null
      SourceMode=AttachParent
      UvSetIndex=0
      WholeMeshLOD=-1
      bExcludeBone=false
      bReadDeformedGeometry=true
      bRequireCurrentFrameData=true
      ComponentTags []
      FilteredBones [hand_r]
      FilteredSockets []
      MeshUserParameter {parameter {name=None, typeDefHandle {registeredTypeIndex=1}, varData []}}
      SamplingRegions []
     }}
    Velocity Sampling=::NewEnumerator0
    VelocitySpeedLimit=false
   )
  ]
  rendererClasses [&NiagaraSpriteRendererProperties]
  renderers [{rendererClass=&NiagaraSpriteRendererProperties}]
 }
}
summary {
 systemName=NS_Emote_TrickyCards_Cards
 emitters [
  {emitterName=Position, rendererClasses [&NiagaraSpriteRendererProperties]}
  {emitterName=Cards001, rendererClasses [&NiagaraMeshRendererProperties]}
  {emitterName=Final_Card_Burst001, rendererClasses [&NiagaraMeshRendererProperties]}
 ]
 userVariables [User.Card_Velocity {type=NiagaraFloat, defaultValue=1, description=""}]
}
